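# Locker collection item card widget, built on the UIKit slim item card base, that displays an item as a list entry.
# It binds the list item's FortItemVM and configures a status indicator, with helpers for equipped slot icons, restricted dimming and quest availability via gameplay tags.
# Hover plays a scale animation except on touch, where focus state follows selection instead.
# A corner context menu button is attached, though its handler functions here are empty.
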WidgetBlueprint /Locker/UIKit/Items/WBP_UIKit_ItemCard_Collection
animations [Anim_Hovered {
  binding_count=1
  length_seconds=0.2004
  object=&self:Anim_Hovered
  sequence {bindings [WBP_UIKit_ItemCard_Locker_C_17 {
     id=#1
     possessed_class=WBP_UIKit_ItemCard_Locker_C
     tracks [{
       type=MovieScene2DTransformTrack
       sections [{
         type=MovieScene2DTransformSection
         end_frame=4
         start_frame=0
         channels [
          Translation.X {type=MovieSceneScriptingFloatChannel, const [0,1]}
          Translation.Y {type=MovieSceneScriptingFloatChannel, const [0,1]}
          Angle {type=MovieSceneScriptingFloatChannel, const [0,1]}
          Scale.X {type=MovieSceneScriptingFloatChannel, keys [[0,1],[4.007,1.03]]}
          Scale.Y {type=MovieSceneScriptingFloatChannel, keys [[0,1],[4.007,1.03]]}
          Shear.X {type=MovieSceneScriptingFloatChannel, const [0,1]}
          Shear.Y {type=MovieSceneScriptingFloatChannel, const [0,1]}
         ]
        }]
      }]
    }]}
 }]
graphs [
 AnalyticsItemInteracted {nodes [
   n1 = AnalyticsItemInteracted!() -> (then, Context:String) {is=title K2Node_FunctionEntry}
   n2 = GetItemTemplateId!(self:Fort Item VM Object Reference=Get FortItemVM() -> (FortItemVM:Fort Item VM Object Reference)) -> (ReturnValue:String)
   > FireEventElementInteractedByName!(
    execute
    Element Name:String="To String (Text)"(InText=Format Text!(Format="LOCTEXT(\"{0} {1}\")", 0:String=@n1, 1:String=@n2) -> (Result))
    WidgetCallingFrom:Widget Object Reference=Self-Reference()
   ) -> (then, Success:bool)
  ]}
 CheckBackgroundType {nodes [
   n1 = CheckBackgroundType!() -> (then, IsNewPassItem:bool=~, Material:Material Interface Soft Object Reference) {is=title K2Node_FunctionEntry}
   > n2 = Branch(Condition=@n1)
   > Return Node!(execute, Background:Material Interface Soft Object Reference=/FortUILibrary/MaterialInstances/Effects/MI_UI_RewardState)
   Return Node(execute=@n2.else, Background=Get Material() -> (Material:Material Interface Soft Object Reference))
  ]}
 Convert_FortItemVMToSelectedConnectedQuestCategoryTag {nodes [
   n1 = Convert_FortItemVMToSelectedConnectedQuestCategoryTag!() -> (then, InItemVM:Fort Item VM Object Reference) {is=title K2Node_FunctionEntry}
   n2 = Get ItemDefinition!(self:Fort Item VM Object Reference=@n1) -> (ItemDefinition:Fort Item Definition Object Reference)
   > Is Valid(InputObject=@n2)
   > Return Node!(
    execute
    OutGameplayTags:Gameplay Tag Structure=GetConnectedQuestCategoryTag!(ItemDefinition:Item Definition Base Object Reference=@n2) -> (ReturnValue:Gameplay Tag Structure)
   )
  ]}
 EventGraph {nodes [
   n1 = Event PreConstruct()
   > Parent: PreConstruct(IsDesignTime=@n1.IsDesignTime)
   n2 = Event OnListItemObjectSet()
   > n3 = Cast To FortItemVM!(execute, Object=@n2.ListItemObject) -> (then, CastFailed, AsFort Item VM:Fort Item VM Object Reference) {is=title K2Node_DynamicCast}
   > InitializeItemCard(self:WBP UIKit Slim Item Card Base Object Reference)
   > SetFortItemVM!(execute, self, Viewmodel:Fort Item VM Object Reference=@n3) -> (then)
   > SetCachedIndicatorConfig!(
    execute
    self
    CachedIndicatorTypeToFind:E_UIKit_ItemCardStatusIndicator Enum=NewEnumerator5
    NewIndicatorConfig:E_UI_StatusIndicator_Configuration Enum=NewEnumerator2
   ) -> (then, Success:bool)
   Event Construct()
   > Parent: Construct()
   Event Destruct()
   > Parent: Destruct()
   n4 = Event OnRemovedFromFocusPath!() -> (OutputDelegate, then, InFocusEvent:Focus Event Structure) {is=title K2Node_Event}
   > Parent: OnRemovedFromFocusPath!(execute, InFocusEvent:Focus Event Structure=@n4.InFocusEvent) -> (then) {is=title K2Node_CallParentFunction}
   Event On Deselected()
   > Branch(Condition=Get IsButtonFocused() -> (IsButtonFocused:bool))
   > Parent: On Deselected() {is=title K2Node_CallParentFunction}
   Event On Hovered()
   > n5 = Branch(Condition=IsUsingTouch(PlayerController=GetOwningPlayer()))
   Set IsItemCardLockerHovered(execute=@n5.else, IsItemCardLockerHovered:bool=true) -> (Output_Get:bool)
   > Parent: On Hovered() {is=title K2Node_CallParentFunction}
   > PlayAnimationForward(InAnimation=Get Anim_Hovered() -> (Anim_Hovered:Widget Animation Object Reference))
   Event On Unhovered()
   > n6 = Branch(Condition=IsUsingTouch(PlayerController=GetOwningPlayer()))
   Set IsItemCardLockerHovered(execute=@n6.else, IsItemCardLockerHovered=false)
   > Parent: On Unhovered() {is=title K2Node_CallParentFunction}
   > PlayAnimationReverse(InAnimation=Get Anim_Hovered())
   n7 = Event On Input Method Changed!() -> (OutputDelegate, then, CurrentInputType:ECommonInputType Enum) {is=title K2Node_Event}
   > Is Valid(InputObject=Get FortItemVM())
   > n8 = Switch on ECommonInputType!(execute, Selection:ECommonInputType Enum=@n7.CurrentInputType) -> (MouseAndKeyboard:Exec, Gamepad:Exec, Touch:Exec, Count:Exec) {is=title K2Node_SwitchEnum}
   > Branch(Condition=GetSelected!(self:Common Button Base Object Reference) -> (ReturnValue:bool))
   > SetFocusState!(execute, self:WBP UIKit Slim Item Card Base Object Reference, IsFocused:bool=false) -> (then)
   Branch(execute=@n8.Touch, Condition=GetSelected())
   > SetFocusState(IsFocused=true)
   n9 = Event OnAddedToFocusPath!() -> (OutputDelegate, then, InFocusEvent:Focus Event Structure) {is=title K2Node_Event}
   > Parent: OnAddedToFocusPath!(execute, InFocusEvent:Focus Event Structure=@n9.InFocusEvent) -> (then) {is=title K2Node_CallParentFunction}
   > n10 = Switch on ECommonInputType(Selection=GetCurrentInputType(self=Get CommonInputSubsystem()))
   SetFocusState(execute=@n10.Touch, IsFocused=GetSelected())
  ]}
 GoToLockerItemDetails {nodes [GoToLockerItemDetails!() -> (then, CustomItemDetails:User Widget Class Reference) {is=title K2Node_FunctionEntry}]}
 Is Item Equipped {nodes [
   n1 = Is Item Equipped!() -> (then, Category Item:Fort Item VM Object Reference, Selected Item:Fort Item VM Object Reference) {is=title K2Node_FunctionEntry}
   > Return Node!(execute, NewParam:bool="Equal (Object)"(A=@n1, B=@n1.Selected Item))
  ]}
 IsModalOpen {nodes [IsModalOpen!() -> (then, IsOpen:bool=~) {is=title K2Node_FunctionEntry}]}
 OnContextMenuClosed {nodes [OnContextMenuClosed() {is=title K2Node_FunctionEntry}]}
 OnContextMenuEditStylesClicked {nodes [OnContextMenuEditStylesClicked() {is=title K2Node_FunctionEntry}]}
 OnContextMenuOpened {nodes [OnContextMenuOpened!() -> (then, ContextMenuWidgetParentEntry:Widget Object Reference) {is=title K2Node_FunctionEntry}]}
 OnItemCardHovered {nodes [OnItemCardHovered() {is=title K2Node_FunctionEntry}]}
 OnItemDoubleClicked {nodes [OnItemDoubleClicked() {is=title K2Node_FunctionEntry}]}
 SelectCard {nodes [
   SelectCard() {is=title K2Node_FunctionEntry}
   > PlayAnimationForward(InAnimation=Get Anim_Hovered())
   > SetFocusState(IsFocused=true)
   > n1 = Cast To ListView!(execute, Object=GetOwningListView!(UserListEntry:User List Entry Interface) -> (ReturnValue:List View Base Object Reference)) -> (then, CastFailed, AsList View:List View Object Reference) {is=title K2Node_DynamicCast}
   > Set Selected Item!(
    execute
    self:List View Object Reference=@n1
    Item:Object Reference=GetListItemObject!(UserObjectListEntry:User Object List Entry Interface) -> (ReturnValue:Object Reference)
   ) -> (then) {is=title K2Node_CallFunction}
  ]}
 SetCachedIndicatorConfig {nodes [
   SetCachedIndicatorConfig!() -> (
    then
    CachedIndicatorTypeToFind:E_UIKit_ItemCardStatusIndicator Enum=NewEnumerator5
    NewIndicatorConfig:E_UI_StatusIndicator_Configuration Enum=NewEnumerator2
   ) {is=title K2Node_FunctionEntry}
   > n1 = FindCachedStatusIndicator!(
    execute
    self:WBP UIKit Slim Item Card Base Object Reference
    IndicatorType:E_UIKit_ItemCardStatusIndicator Enum=Get CachedIndicatorTypeToFind() -> (CachedIndicatorTypeToFind:E_UIKit_ItemCardStatusIndicator Enum=NewEnumerator10)
   ) -> (then, Status Indicator:WBP UIKit Status Indicator Base Object Reference, Valid:bool)
   > n2 = Branch(Condition=@n1.Valid)
   > SetIndicatorConfiguration!(
    execute
    self:WBP UIKit Status Indicator Base Object Reference=@n1
    InCount:int=0
    IndicatorConfig:E_UI_StatusIndicator_Configuration Enum=Get NewIndicatorConfig() -> (NewIndicatorConfig:E_UI_StatusIndicator_Configuration Enum=NewEnumerator0)
   ) -> (then)
   > Return Node!(execute, Success:bool=true)
   Return Node(execute=@n2.else, Success=false)
  ]}
 SetItemIsRestricted {nodes [
   n1 = SetItemIsRestricted!() -> (then, IsCompatible:bool) {is=title K2Node_FunctionEntry}
   > Parent: SetItemIsRestricted!(execute, IsCompatible:bool=@n1) -> (then) {is=title K2Node_CallParentFunction}
   > SetOpacity!(
    execute
    self:Image Object Reference=Get Background() -> (Background:Image Object Reference)
    InOpacity:float=SelectFloat(A=0.25, B=1, bPickA=@n1)
   ) -> (then)
   Self-Reference()
  ]}
 Update Equipped Indexes {nodes [
   n1 = Update Equipped Indexes!() -> (then, Equipped Indexes:Array of Integers) {is=title K2Node_FunctionEntry}
   > n2 = Branch(Condition=Is Empty!(TargetArray:Array of Integers=@n1) -> (ReturnValue:bool))
   > SetItemIsEquippedSecondary!(
    execute
    self:WBP UIKit Slim Item Card Base Object Reference
    Is Equipped:bool=false
    ShowEquippedSlotIndex:bool=false
    EquippedSlotIndex:int=0
    SlotIcon:Soft Object Reference
   ) -> (then)
   n3 = GetItemCategory!(self:Fort Item VM Object Reference=Get FortItemVM()) -> (ReturnValue:Fort Locker Category Config Object Reference)
   n4 = "Get (a copy)"(Array:Array of Integers=Get Equipped Indexes() -> (Equipped Indexes:Array of Integers)) -> (Output:int)
   SetItemIsEquippedSecondary(
    execute=@n2.else
    Is Equipped=true
    ShowEquippedSlotIndex=Get bSecondaryEquipShowNumbers!(self:Fort Locker Category Config Object Reference=@n3) -> (bSecondaryEquipShowNumbers:bool)
    EquippedSlotIndex=int + int(A=@n4)
    SlotIcon=Break Fort Locker Category Slot Info!(FortLockerCategorySlotInfo:"Fort Locker Category Slot Info Structure (by ref)"=GetSlotInfo!(self:Fort Locker Category Config Object Reference=@n3, SlotIndex:int=@n4) -> (ReturnValue:"Fort Locker Category Slot Info Structure (by ref)")) -> (
     CategoryName:Text
     CategoryNamePlural:Text
     CategoryNameWithPluralForms:Text
     CategoryDescription:Text
     DefaultEmptyTexture:Soft Object Reference
     OverrideShuffleTexture:Texture 2D Soft Object Reference
     CategoryIcon:Material Instance Object Reference
     CategorySoundOverrides:Map of Gameplay Tag Structures to Sound Base Soft Object References
     bShowInCategoryTab:bool=~
     PreviewWindowScale:float=~
     PreviewWindowOffset:Vector 3f
     bUsePluralCategoryInItemList:bool=~
    ) {is=title K2Node_BreakStruct}.DefaultEmptyTexture
   )
  ]}
 VM_UpdateIsQuestAvailable {nodes [
   n1 = VM_UpdateIsQuestAvailable!() -> (then, InGameplayTags:Gameplay Tag Structure) {is=title K2Node_FunctionEntry}
   > n2 = Cast To FortPlayerController(Object=GetOwningPlayer())
   > n3 = PlayerHasQuestsForCategory!(execute, FortPC:Fort Player Controller Object Reference=@n2, QuestCategoryTag:"Gameplay Tag Structure (by ref)"=@n1) -> (then, ReturnValue:bool)
   > Set IsQuestAvailable(IsQuestAvailable:bool=AND Boolean(A=@n3, B=IsGameplayTagValid!(GameplayTag:Gameplay Tag Structure=@n1) -> (ReturnValue:bool))) -> (Output_Get:bool)
  ]}
]
variables [
 Favorite Sound {type=SoundBase, default=/Game/Sounds/Fort_UI_Sounds/FrontEnd/MultiProduct/Common/Buttons/MSS_UI_MP_Button_Click_Favorite}
 GoToLockerItemDetails {}
 IsEmoteWheel {type=boolean}
 IsItemCardLockerHovered {type=boolean}
 IsModalOpen {}
 IsQuestAvailable {type=boolean}
 OnContextMenuClosed {}
 OnContextMenuEditStylesClicked {}
 OnContextMenuOpened {}
 OnItemCardHovered {}
 OnItemDoubleClicked {}
 Right Clicked Sound {type=SoundBase, default=/Game/Sounds/Fort_UI_Sounds/FrontEnd/DefaultSounds/PressedButtons/MSS_UI_Button_Pressed_Default}
 Unfavorite Sound {type=SoundBase, default=/Game/Sounds/Fort_UI_Sounds/FrontEnd/MultiProduct/Common/Buttons/MSS_UI_MP_Button_Click_Unfavorite}
]
widgets {
 info {
  inheritedWidgetCount=0
  namedSlotCount=1
  parentClass=/FortUILibrary/Systems/Items/WBP_UIKit_SlimItemCard_Base.WBP_UIKit_SlimItemCard_Base_C
  rootWidgetClass=null
  widgetCount=2
 }
 widgets [
  ContextMenu_Attachment {
   widgetClassPath=&FortVisualAttachment
   bIsVariable=true
   widget=ContextMenu_Attachment
   properties {bOverride_HeightOverride=true, bOverride_WidthOverride=true, heightOverride=40, visibility=Visible, widthOverride=40, contentAnchor=<1 1>}
   slot=~
  }
  ContextMenu {
   widgetClassPath=/FortUILibrary/UIKit/ContextMenu/WBP_UIKit_ContextMenuAnchorButton
   bIsVariable=true
   parent=ContextMenu_Attachment
   widget=ContextMenu
   properties {
    button Background Material=/FortUILibrary/UIKit/Buttons/MaterialInstances/CTAButtons/MI_UIKit_BgQuiet
    buttonClass=/FortUILibrary/UIKit/Buttons/WBP_UIKit_Button_Contextual.WBP_UIKit_Button_Contextual_C
    buttonIcon=/FortUILibrary/Textures/Icons/T_UI_Icon_UIKit_ContextMenu
    buttonText=""
    button Padding=<ltrb 4 4 4 4>
    button Text Font {
     bForceMonospaced=false
     bMaterialIsStencil=false
     fontMaterial=/FortUILibrary/UIKit/Buttons/MaterialInstances/GenericButtons/MI_UIKit_GenericText_Quiet
     fontObject=&Burbank
     letterSpacing=180
     monospacedWidth=1
     size=26.25
     skewAmount=0
     typefaceFontName=BurbankRegular-Black
     outlineSettings {
      bApplyOutlineToDropShadows=false
      bMiteredCorners=false
      bSeparateFillAlpha=false
      outlineMaterial=null
      outlineSize=0
      outlineColor=<rgba 0 0 0 1>
     }
    }
   }
   slot {class=SizeBoxSlot, properties {horizontalAlignment=HAlign_Right, verticalAlignment=VAlign_Bottom, padding=<ltrb -8 -8 8 8>}}
  }
 ]
}
